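# Animation Blueprint for Grandma, a quadruped predator AI that can walk, run, jump, swim, sleep and be ridden.
# A bool switches between normal Locomotion with alert/sniff additives and RidingLocomotion with a landing additive, cached as PostLocoSM.
# Montage slots, an abduction override and face-only blinking layer over that pose.
# Weighted look-at, quadruped slope warping, curve-driven four-leg IK and inertialization finish it.
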AnimBlueprint /Irwin/AI/Predators/Grandma/Animations/Grandma_AnimBP
target_skeleton=/Irwin/AI/Base/Grandma_Mammal_Invis_Skeleton
graphs [
 AnimGraph {nodes [
   n1 = Sequence Player 'Grandma_Blink'() {sequence=&/Grandma_Blink}
   n2 = Layered blend per bone(
    BasePose=Use cached pose 'PostLocoSM'()
    BlendPoses_0=Slot 'UpperBodyInMotion'(Source=Use cached pose 'PostLocoSM'()) {slotName=UpperBodyInMotion}
   ) {
    bMeshSpaceRotationBlend=true
    layerSetup [{branchFilters [{blendDepth=5, boneName=C_QuadSpine_A_Spine1_Jnt}, {blendDepth=2, boneName=IK_Foot_Front_R}, {blendDepth=2, boneName=IK_Foot_Front_L}]}]
   }
   n3 = Blend Poses by bool(
    BlendPose_0=Slot 'FullBodyInPlace'(Source=Use cached pose 'PostLocoSM'()) {slotName=FullBodyInPlace}
    BlendPose_1=@n2
    BlendTime_0=0.4
    BlendTime_1=0.4
    bActiveValue=Get bShouldPlayFullBodyInPlace() -> (bShouldPlayFullBodyInPlace:bool)
   ) {blendType=Linear, transitionType=Inertialization}
   n4 = Sequence Player 'Grandma_Abduction'() {playRate=1.2, sequence=&/Grandma_Abduction}
   n5 = Get bIsAbducted() -> (bIsAbducted:bool)
   n6 = Slot 'FullBodyAdditive'(Source=Slot 'DefaultSlot'(Source=Blend Poses by bool(BlendPose_0=@n4, BlendPose_1=@n3, BlendTime_0=0.4, BlendTime_1=0.4, bActiveValue=@n5) {blendType=Linear, transitionType=Inertialization})) {slotName=FullBodyAdditive}
   n7 = Get BlinkAlpha() -> (BlinkAlpha:float)
   n8 = Weighted Look At - Bone: BipedHeadNeck_A_Head_C!(
    ComponentPose:Component Space Pose Link Structure=Local To Component(LocalPose=Layered blend per bone(BasePose=@n6, BlendPoses_0=@n1, BlendWeights_0=@n7) {layerSetup [{branchFilters [{boneName=faceAttach}]}]})
    Alpha:float=Get LookAtAlpha() -> (LookAtAlpha:float)
    LookAtLocation:Vector=Get AimTargetLocation() -> (AimTargetLocation:Vector)
   ) -> (Pose:Component Space Pose Link Structure) {
    is=title FortAnimGraphNode_WeightedLookAt
    interpolationTime=3
    interpolationType=EaseInOutExponent3
    lookAtClamp=60
    alphaBoolBlend {blendOption=Linear}
    boneToModify {boneName=BipedHeadNeck_A_Head_C}
    boneWeightsToApply [{weight=0.3}, {weight=0.3}, {weight=0.25}, {weight=0.15}]
    lookAt_Axis {bInLocalSpace=false}
    lookUp_Axis {axis {y=1}}
   }
   n9 = Quadruped Slope Warping!(ComponentPose:Component Space Pose Link Structure=@n8, AlphaCurveName:Name=BothLegIK) -> (Pose:Component Space Pose Link Structure) {
    is=title FortAnimGraphNode_QuadrupedSlopeWarping
    alphaInputType=Curve
    numSpineBones=4
    alphaBoolBlend {blendOption=Linear}
    alphaScaleBiasClamp {bias=1, scale=-1}
    armsFeetDefinitions [
     {numBonesInLimb=3, fKFootBone {boneName=PawedArm_A_Ball_L}, iKFootBone {boneName=IK_Foot_Front_L}}
     {numBonesInLimb=3, fKFootBone {boneName=PawedArm_A_Ball_R}, iKFootBone {boneName=IK_Foot_Front_R}}
    ]
    armsIKFootRootBone {boneName=IK_Foot_Front_Root}
    armsRootBone {boneName=QuadSpine_A_Chest_C}
    feetDefinitions [
     {numBonesInLimb=3, fKFootBone {boneName=PawedLeg_A_Ball_L}, iKFootBone {boneName=IK_Foot_Rear_L}}
     {numBonesInLimb=3, fKFootBone {boneName=PawedLeg_A_Ball_R}, iKFootBone {boneName=IK_Foot_Rear_R}}
    ]
    iKFootRootBone {boneName=IK_Foot_Rear_Root}
    pelvisBone {boneName=Leaf_Hips_Leaf_C}
    skeletonRootBone {boneName=root}
   }
   n10 = Leg IK!(ComponentPose:Component Space Pose Link Structure=@n9, AlphaCurveName:Name=BothLegIK) -> (Pose:Component Space Pose Link Structure) {
    is=title AnimGraphNode_LegIK
    alphaInputType=Curve
    alphaBoolBlend {blendOption=Linear}
    alphaScaleBiasClamp {bias=1, scale=-1}
    legsDefinition [
     {
      bEnableKneeTwistCorrection=true
      bEnableRotationLimit=false
      footBoneForwardAxis=Y
      hingeRotationAxis=None
      minRotationAngle=15
      numBonesInLimb=2
      twistOffsetCurveName=None
      fKFootBone {boneName=PawedLeg_A_Ball_L}
      iKFootBone {boneName=IK_Foot_Rear_L}
     }
     {
      bEnableKneeTwistCorrection=true
      bEnableRotationLimit=false
      footBoneForwardAxis=Y
      hingeRotationAxis=None
      minRotationAngle=15
      numBonesInLimb=2
      twistOffsetCurveName=None
      fKFootBone {boneName=PawedLeg_A_Ball_R}
      iKFootBone {boneName=IK_Foot_Rear_R}
     }
     {
      bEnableKneeTwistCorrection=true
      bEnableRotationLimit=false
      footBoneForwardAxis=Y
      hingeRotationAxis=None
      minRotationAngle=15
      numBonesInLimb=2
      twistOffsetCurveName=None
      fKFootBone {boneName=PawedArm_A_Ball_R}
      iKFootBone {boneName=IK_Foot_Front_R}
     }
     {
      bEnableKneeTwistCorrection=true
      bEnableRotationLimit=false
      footBoneForwardAxis=Y
      hingeRotationAxis=None
      minRotationAngle=15
      numBonesInLimb=2
      twistOffsetCurveName=None
      fKFootBone {boneName=PawedArm_A_Ball_L}
      iKFootBone {boneName=IK_Foot_Front_L}
     }
    ]
   }
   n11 = Component To Local(ComponentPose=@n10)
   Output Pose(Result=Inertialization!(Source:Pose Link Structure=Slot 'SequencerDefault'(Source=@n11) {slotName=SequencerDefault}) -> (Pose:Pose Link Structure) {is=title AnimGraphNode_Inertialization}) {layerGroup=None}
   AnimGraphNode_StateMachine_0 = EntryAdditive() {is=title AnimGraphNode_StateMachine}
   AnimGraphNode_StateMachine_3 = RidingLocomotion() {is=title AnimGraphNode_StateMachine}
   AnimGraphNode_StateMachine_2 = Locomotion() {is=title AnimGraphNode_StateMachine}
   n12 = Slot 'FullBodyPreAdditivePose'(Source=@AnimGraphNode_StateMachine_2) {slotName=FullBodyPreAdditivePose}
   AnimGraphNode_StateMachine_1 = LocomotionPoseAdditives() {is=title AnimGraphNode_StateMachine}
   n13 = Get LocomotionPoseAdditiveAlpha() -> (LocomotionPoseAdditiveAlpha:float)
   n14 = Get RidingAnimationData() -> (RidingAnimationData:Riding Animation Data Structure)
   Save cached pose 'PostLocoSM'(Pose=Blend Poses by bool(
     BlendPose_0=Apply Additive(Base=@AnimGraphNode_StateMachine_3, Additive=@AnimGraphNode_StateMachine_0) {alphaBoolBlend {blendOption=Linear}}
     BlendPose_1=Apply Additive(Base=@n12, Additive=@AnimGraphNode_StateMachine_1, Alpha=@n13) {alphaBoolBlend {blendOption=Linear}}
     bActiveValue=Break Riding Animation Data!(RidingAnimationData:"Riding Animation Data Structure (by ref)"=@n14) -> (
      bIsRidden:bool=~
      bIsTurningInPlace:bool=~
      bIsSprinting:bool=~
      bIsBeingPetted:bool=~
      bIsMovingBackwards:bool=~
      TurnInPlaceRotationSpeed:float=~
      TurnInPlaceAngleDelta:float=~
      RiderReferentialYaw:float=~
     ) {is=title K2Node_BreakStruct}
    ) {blendType=Linear}) {cacheName=PostLocoSM}
   Save cached pose 'FaunaIdlePose'(Pose=Fauna_Idle_AnimInterface - Fauna Idle!(InPose Idle:Pose Link Structure) -> (Pose:Pose Link Structure) {is=title AnimGraphNode_LinkedAnimLayer, interface=/Irwin/AI/Interfaces/Fauna_Idle_AnimInterface.Fauna_Idle_AnimInterface_C, layer=Fauna Idle}) {cacheName=FaunaIdlePose}
  ]}
 EntryAdditive {
  entry=AddRef
  path=AnimGraph.AnimGraphNode_StateMachine_0.EntryAdditive
  ids=id
  nodes [
   AnimStateNode_0 = AddRef() {is=title AnimStateNode}
   AnimStateNode_1 = EntryAdd() {is=title AnimStateNode, stateLeft {guid=#1, notifyName=LandAddExit}}
   AnimStateTransitionNode_0 = AnimStateTransitionNode() {crossfadeDuration=0.1, logicType=TLT_Inertialization}
   AnimStateTransitionNode_1 = AnimStateTransitionNode() {bAutomaticRuleBasedOnSequencePlayerInState=true, crossfadeDuration=0.3, logicType=TLT_Inertialization}
  ]
  links [
   {
    dst=EntryAdd
    rule=AnimGraph.AnimGraphNode_StateMachine_0.EntryAdditive.AnimStateTransitionNode_0.Transition
    src=AddRef
    via=AnimStateTransitionNode_0
   }
   {
    dst=AddRef
    rule=AnimGraph.AnimGraphNode_StateMachine_0.EntryAdditive.AnimStateTransitionNode_1.Transition
    src=EntryAdd
    via=AnimStateTransitionNode_1
   }
  ]
 }
 AddRef {
  path=AnimGraph.AnimGraphNode_StateMachine_0.EntryAdditive.AnimStateNode_0.AddRef
  nodes [Output Animation Pose(Result=Additive Identity Pose())]
 }
 EntryAdd {
  path=AnimGraph.AnimGraphNode_StateMachine_0.EntryAdditive.AnimStateNode_1.EntryAdd
  nodes [Output Animation Pose(Result="Sequence Player (additive) 'Grandma_JumpLand_Moving_Add'"() {sequence=&/Grandma_JumpLand_Moving_Add})]
 }
 Transition {path=AnimGraph.AnimGraphNode_StateMachine_0.EntryAdditive.AnimStateTransitionNode_0.Transition, nodes [Result()]}
 Transition {path=AnimGraph.AnimGraphNode_StateMachine_0.EntryAdditive.AnimStateTransitionNode_1.Transition, nodes [Result()]}
 LocomotionPoseAdditives {
  entry=Default/IdentityPose
  path=AnimGraph.AnimGraphNode_StateMachine_1.LocomotionPoseAdditives
  ids=id
  nodes [
   AnimStateNode_0 = Default/IdentityPose() {is=title AnimStateNode}
   AnimStateNode_1 = Alert() {is=title AnimStateNode}
   AnimStateNode_2 = Sniff() {is=title AnimStateNode}
   AnimStateTransitionNode_2 = AnimStateTransitionNode() {crossfadeDuration=1, logicType=TLT_Inertialization}
   AnimStateTransitionNode_3 = AnimStateTransitionNode() {crossfadeDuration=1, logicType=TLT_Inertialization}
   AnimStateTransitionNode_4 = AnimStateTransitionNode() {bAutomaticRuleBasedOnSequencePlayerInState=true, crossfadeDuration=1, logicType=TLT_Inertialization}
   AnimStateTransitionNode_5 = AnimStateTransitionNode() {bAutomaticRuleBasedOnSequencePlayerInState=true, crossfadeDuration=0.8, logicType=TLT_Inertialization}
  ]
  links [
   {
    dst=Default/IdentityPose
    rule=AnimGraph.AnimGraphNode_StateMachine_1.LocomotionPoseAdditives.AnimStateTransitionNode_4.Transition
    src=Alert
    via=AnimStateTransitionNode_4
   }
   {
    dst=Alert
    rule=AnimGraph.AnimGraphNode_StateMachine_1.LocomotionPoseAdditives.AnimStateTransitionNode_2.Transition
    src=Default/IdentityPose
    via=AnimStateTransitionNode_2
   }
   {
    dst=Sniff
    rule=AnimGraph.AnimGraphNode_StateMachine_1.LocomotionPoseAdditives.AnimStateTransitionNode_3.Transition
    src=Default/IdentityPose
    via=AnimStateTransitionNode_3
   }
   {
    dst=Default/IdentityPose
    rule=AnimGraph.AnimGraphNode_StateMachine_1.LocomotionPoseAdditives.AnimStateTransitionNode_5.Transition
    src=Sniff
    via=AnimStateTransitionNode_5
   }
  ]
 }
 Default/IdentityPose {
  path=AnimGraph.AnimGraphNode_StateMachine_1.LocomotionPoseAdditives.AnimStateNode_0.Default/IdentityPose
  nodes [Output Animation Pose(Result=Additive Identity Pose())]
 }
 Alert {
  path=AnimGraph.AnimGraphNode_StateMachine_1.LocomotionPoseAdditives.AnimStateNode_1.Alert
  nodes [Output Animation Pose(Result="Sequence Player (additive) 'Grandma_LocomotionAdditive_Alert'"() {sequence=&/Grandma_LocomotionAdditive_Alert})]
 }
 Sniff {
  path=AnimGraph.AnimGraphNode_StateMachine_1.LocomotionPoseAdditives.AnimStateNode_2.Sniff
  nodes [Output Animation Pose(Result="Sequence Player (additive) 'Grandma_LocomotionAdditive_Sniff'"() {sequence=&/Grandma_LocomotionAdditive_Sniff})]
 }
 Transition {
  path=AnimGraph.AnimGraphNode_StateMachine_1.LocomotionPoseAdditives.AnimStateTransitionNode_2.Transition
  nodes [Result(bCanEnterTransition=Get bEnableAlertAdditive() -> (bEnableAlertAdditive:bool))]
 }
 Transition {
  path=AnimGraph.AnimGraphNode_StateMachine_1.LocomotionPoseAdditives.AnimStateTransitionNode_3.Transition
  nodes [Result(bCanEnterTransition=Get bEnableSniffAdditive() -> (bEnableSniffAdditive:bool))]
 }
 Transition {path=AnimGraph.AnimGraphNode_StateMachine_1.LocomotionPoseAdditives.AnimStateTransitionNode_4.Transition, nodes [Result()]}
 Transition {path=AnimGraph.AnimGraphNode_StateMachine_1.LocomotionPoseAdditives.AnimStateTransitionNode_5.Transition, nodes [Result()]}
 Locomotion {
  entry=Idle
  path=AnimGraph.AnimGraphNode_StateMachine_2.Locomotion
  ids=id
  nodes [
   AnimStateNode_0 = Idle() {
    stateEntered {guid=#2}
    stateFullyBlended {guid=#3}
    stateLeft {guid=#4}
   }
   AnimStateNode_1 = Moving() {
    is=title AnimStateNode
    stateEntered {guid=#5}
    stateFullyBlended {guid=#6}
    stateLeft {guid=#7}
   }
   AnimStateNode_2 = JumpStart() {is=title AnimStateNode}
   AnimStateNode_3 = FallAsleep() {is=title AnimStateNode}
   AnimStateNode_4 = Sleeping() {is=title AnimStateNode}
   AnimStateNode_5 = Fall() {is=title AnimStateNode, stateLeft {guid=#8, notifyName=FallExit}}
   AnimStateNode_6 = JumpLand() {is=title AnimStateNode, stateEntered {guid=#9, notifyName=LandEnter}}
   AnimStateNode_7 = WakeUp() {is=title AnimStateNode}
   AnimStateNode_9 = Swimming() {is=title AnimStateNode}
   AnimStateTransitionNode_0 = AnimStateTransitionNode() {crossfadeDuration=0.15, logicType=TLT_Inertialization}
   AnimStateTransitionNode_1 = AnimStateTransitionNode() {crossfadeDuration=0.4, logicType=TLT_Inertialization, priorityOrder=3}
   AnimStateTransitionNode_10 = AnimStateTransitionNode()
   AnimStateTransitionNode_11 = AnimStateTransitionNode()
   AnimStateTransitionNode_13 = AnimStateTransitionNode() {crossfadeDuration=0.4, logicType=TLT_Inertialization, priorityOrder=100}
   AnimStateTransitionNode_14 = AnimStateTransitionNode() {crossfadeDuration=0.3, logicType=TLT_Inertialization, priorityOrder=100}
   AnimStateTransitionNode_2 = AnimStateTransitionNode()
   AnimStateTransitionNode_29 = AnimStateTransitionNode() {logicType=TLT_Inertialization, priorityOrder=3}
   AnimStateTransitionNode_3 = AnimStateTransitionNode()
   AnimStateTransitionNode_30 = AnimStateTransitionNode() {crossfadeDuration=0.4, logicType=TLT_Inertialization}
   AnimStateTransitionNode_37 = AnimStateTransitionNode() {crossfadeDuration=0.55, logicType=TLT_Inertialization, priorityOrder=2}
   AnimStateTransitionNode_38 = AnimStateTransitionNode() {crossfadeDuration=0.25, logicType=TLT_Inertialization, priorityOrder=100}
   AnimStateTransitionNode_4 = AnimStateTransitionNode() {bAutomaticRuleBasedOnSequencePlayerInState=true, crossfadeDuration=0.4, logicType=TLT_Inertialization}
   AnimStateTransitionNode_40 = AnimStateTransitionNode() {crossfadeDuration=0.4, logicType=TLT_Inertialization, priorityOrder=100}
   AnimStateTransitionNode_41 = AnimStateTransitionNode() {crossfadeDuration=0.35, logicType=TLT_Inertialization}
   AnimStateTransitionNode_5 = AnimStateTransitionNode()
   AnimStateTransitionNode_6 = AnimStateTransitionNode() {crossfadeDuration=0.3, logicType=TLT_Inertialization}
   AnimStateTransitionNode_7 = AnimStateTransitionNode() {logicType=TLT_Inertialization, priorityOrder=2}
   AnimStateTransitionNode_8 = AnimStateTransitionNode() {logicType=TLT_Inertialization, priorityOrder=100}
   AnimStateTransitionNode_9 = AnimStateTransitionNode() {crossfadeDuration=0.3, logicType=TLT_Inertialization, priorityOrder=2}
  ]
  links [
   {
    dst=JumpLand
    rule=AnimGraph.AnimGraphNode_StateMachine_2.Locomotion.AnimStateTransitionNode_8.AnimationTransitionGraph_0
    src=Fall
    via=AnimStateTransitionNode_8
   }
   {
    dst=Moving
    rule=AnimGraph.AnimGraphNode_StateMachine_2.Locomotion.AnimStateTransitionNode_0.Transition
    src=Fall
    via=AnimStateTransitionNode_0
   }
   {
    dst=Swimming
    rule=AnimGraph.AnimGraphNode_StateMachine_2.Locomotion.AnimStateTransitionNode_40.Transition
    src=Fall
    via=AnimStateTransitionNode_40
   }
   {
    dst=Sleeping
    rule=AnimGraph.AnimGraphNode_StateMachine_2.Locomotion.AnimStateTransitionNode_5.Transition
    src=FallAsleep
    via=AnimStateTransitionNode_5
   }
   {
    dst=FallAsleep
    rule=AnimGraph.AnimGraphNode_StateMachine_2.Locomotion.AnimStateTransitionNode_3.Transition
    src=Idle
    via=AnimStateTransitionNode_3
   }
   {
    dst=JumpStart
    rule=AnimGraph.AnimGraphNode_StateMachine_2.Locomotion.AnimStateTransitionNode_38.AnimationTransitionGraph_0
    src=Idle
    via=AnimStateTransitionNode_38
   }
   {
    dst=Moving
    rule=AnimGraph.AnimGraphNode_StateMachine_2.Locomotion.AnimStateTransitionNode_37.AnimationTransitionGraph_0
    src=Idle
    via=AnimStateTransitionNode_37
   }
   {
    dst=Swimming
    rule=AnimGraph.AnimGraphNode_StateMachine_2.Locomotion.AnimStateTransitionNode_2.Transition
    src=Idle
    via=AnimStateTransitionNode_2
   }
   {
    dst=Idle
    rule=AnimGraph.AnimGraphNode_StateMachine_2.Locomotion.AnimStateTransitionNode_14.Transition
    src=JumpLand
    via=AnimStateTransitionNode_14
   }
   {
    dst=JumpStart
    rule=AnimGraph.AnimGraphNode_StateMachine_2.Locomotion.AnimStateTransitionNode_6.AnimationTransitionGraph_0
    src=JumpLand
    via=AnimStateTransitionNode_6
   }
   {
    dst=Moving
    rule=AnimGraph.AnimGraphNode_StateMachine_2.Locomotion.AnimStateTransitionNode_7.Transition
    src=JumpLand
    via=AnimStateTransitionNode_7
   }
   {
    dst=Swimming
    rule=AnimGraph.AnimGraphNode_StateMachine_2.Locomotion.AnimStateTransitionNode_1.Transition
    src=JumpLand
    via=AnimStateTransitionNode_1
   }
   {
    dst=Fall
    rule=AnimGraph.AnimGraphNode_StateMachine_2.Locomotion.AnimStateTransitionNode_4.Transition
    src=JumpStart
    via=AnimStateTransitionNode_4
   }
   {
    dst=Fall
    rule=AnimGraph.AnimGraphNode_StateMachine_2.Locomotion.AnimStateTransitionNode_9.Transition
    src=Moving
    via=AnimStateTransitionNode_9
   }
   {
    dst=Idle
    rule=AnimGraph.AnimGraphNode_StateMachine_2.Locomotion.AnimStateTransitionNode_13.Transition
    src=Moving
    via=AnimStateTransitionNode_13
   }
   {
    dst=JumpStart
    rule=AnimGraph.AnimGraphNode_StateMachine_2.Locomotion.AnimStateTransitionNode_41.AnimationTransitionGraph_0
    src=Moving
    via=AnimStateTransitionNode_41
   }
   {
    dst=Swimming
    rule=AnimGraph.AnimGraphNode_StateMachine_2.Locomotion.AnimStateTransitionNode_29.Transition
    src=Moving
    via=AnimStateTransitionNode_29
   }
   {
    dst=WakeUp
    rule=AnimGraph.AnimGraphNode_StateMachine_2.Locomotion.AnimStateTransitionNode_10.Transition
    src=Sleeping
    via=AnimStateTransitionNode_10
   }
   {
    dst=Moving
    rule=AnimGraph.AnimGraphNode_StateMachine_2.Locomotion.AnimStateTransitionNode_30.Transition
    src=Swimming
    via=AnimStateTransitionNode_30
   }
   {
    dst=Idle
    rule=AnimGraph.AnimGraphNode_StateMachine_2.Locomotion.AnimStateTransitionNode_11.Transition
    src=WakeUp
    via=AnimStateTransitionNode_11
   }
  ]
 }
 Idle {
  path=AnimGraph.AnimGraphNode_StateMachine_2.Locomotion.AnimStateNode_0.Idle
  nodes [
   n1 = Blendspace Player 'Grandma_Walk_N_BS_TurnPose'!(X=Get IdleShuffleBodyBend() -> (IdleShuffleBodyBend:float)) -> (Pose) {blendSpace=&/Grandma_Walk_N_BS_TurnPose}
   n2 = Get IdleShufflePlayRate() -> (IdleShufflePlayRate:float)
   n3 = Layered blend per bone(
    BasePose=Use cached pose 'FaunaIdlePose'()
    BlendPoses_0=Blend Poses by bool(
     BlendPose_0=Sequence Player 'Grandma_TurnInPlace_R_90_HipsAndFeet'!(PlayRate:float=@n2) -> (Pose) {sequence=&/Grandma_TurnInPlace_R_90_HipsAndFeet}
     BlendPose_1=Sequence Player 'Grandma_TurnInPlace_L_90_HipsAndFeet'!(PlayRate:float=@n2) -> (Pose) {sequence=&/Grandma_TurnInPlace_L_90_HipsAndFeet}
     BlendTime_0=0.5
     BlendTime_1=0.5
     bActiveValue=Get bIdleShuffleTurnDirection() -> (bIdleShuffleTurnDirection:bool)
    ) {blendType=Linear, transitionType=Inertialization}
   ) {layerSetup [{branchFilters [{boneName=Leaf_Hips_Leaf_C}, {boneName=IK_Foot_Rear_Root}, {boneName=QuadSpine_A_Chest_C}]}]}
   n4 = Get bLookAtAdditiveTurnAlpha() -> (bLookAtAdditiveTurnAlpha:bool)
   Output Animation Pose(Result=Blend Poses by bool(
     BlendPose_0=Apply Additive!(Base=@n3, Additive=@n1, bAlphaBoolEnabled:bool=@n4) -> (Pose) {alphaInputType=Bool, alphaBoolBlend {blendInTime=0.4, blendOption=Linear, blendOutTime=0.4}}
     BlendPose_1=Use cached pose 'FaunaIdlePose'()
     BlendTime_0=0.6
     BlendTime_1=0.4
     bActiveValue=Get bIsIdleShuffleTurn() -> (bIsIdleShuffleTurn:bool)
    ) {blendType=Linear, transitionType=Inertialization})
  ]
 }
 Moving {
  path=AnimGraph.AnimGraphNode_StateMachine_2.Locomotion.AnimStateNode_1.Moving
  nodes [
   n1 = Blendspace Player 'Grandma_WalkRun_BS'!(X=Get WalkRunPlayRate() -> (WalkRunPlayRate:float)) -> (Pose) {blendSpace=&/Grandma_WalkRun_BS, groupName=Locomotion, method=SyncGroup}
   n2 = Blendspace Player 'Grandma_Walk_N_BS_TurnPose'(X=BreakRotator(InRot=Get PawnRotationRate() -> (PawnRotationRate:Rotator)).Yaw) {blendSpace=&Grandma_Walk_N_BS_TurnPose}
   n3 = Get bLookAtAdditiveTurnAlpha()
   AnimGraphNode_StateMachine_0 = JumpLand() {is=title AnimGraphNode_StateMachine}
   Output Animation Pose(Result=Apply Additive!(
     Base=Apply Additive!(Base=@n1, Additive=@n2, bAlphaBoolEnabled:bool=@n3) -> (Pose) {alphaInputType=Bool, alphaBoolBlend {blendInTime=0.35, blendOption=Linear, blendOutTime=0.35}}
     Additive=@AnimGraphNode_StateMachine_0
     Alpha:float
    ) -> (Pose) {alphaBoolBlend {blendOption=Linear}})
  ]
 }
 JumpLand {
  entry=AddRef
  path=AnimGraph.AnimGraphNode_StateMachine_2.Locomotion.AnimStateNode_1.Moving.AnimGraphNode_StateMachine_0.JumpLand
  ids=id
  nodes [
   AnimStateNode_0 = AddRef()
   AnimStateNode_1 = LandAdd() {is=title AnimStateNode, stateLeft {guid=#1, notifyName=LandAddExit}}
   AnimStateTransitionNode_0 = AnimStateTransitionNode() {crossfadeDuration=0.1, logicType=TLT_Inertialization}
   AnimStateTransitionNode_1 = AnimStateTransitionNode() {bAutomaticRuleBasedOnSequencePlayerInState=true, crossfadeDuration=0.3, logicType=TLT_Inertialization}
  ]
  links [
   {
    dst=LandAdd
    rule=AnimGraph.AnimGraphNode_StateMachine_2.Locomotion.AnimStateNode_1.Moving.AnimGraphNode_StateMachine_0.JumpLand.AnimStateTransitionNode_0.Transition
    src=AddRef
    via=AnimStateTransitionNode_0
   }
   {
    dst=AddRef
    rule=AnimGraph.AnimGraphNode_StateMachine_2.Locomotion.AnimStateNode_1.Moving.AnimGraphNode_StateMachine_0.JumpLand.AnimStateTransitionNode_1.Transition
    src=LandAdd
    via=AnimStateTransitionNode_1
   }
  ]
 }
 AddRef {
  path=AnimGraph.AnimGraphNode_StateMachine_2.Locomotion.AnimStateNode_1.Moving.AnimGraphNode_StateMachine_0.JumpLand.AnimStateNode_0.AddRef
  nodes [Output Animation Pose(Result=Additive Identity Pose())]
 }
 LandAdd {
  path=AnimGraph.AnimGraphNode_StateMachine_2.Locomotion.AnimStateNode_1.Moving.AnimGraphNode_StateMachine_0.JumpLand.AnimStateNode_1.LandAdd
  nodes [Output Animation Pose(Result="Sequence Player (additive) 'Grandma_JumpLand_Moving_Add'"() {sequence=&Grandma_JumpLand_Moving_Add})]
 }
 Transition {
  path=AnimGraph.AnimGraphNode_StateMachine_2.Locomotion.AnimStateNode_1.Moving.AnimGraphNode_StateMachine_0.JumpLand.AnimStateTransitionNode_0.Transition
  nodes [Result(bCanEnterTransition=Get bPlayLandAdditive() -> (bPlayLandAdditive:bool))]
 }
 Transition {
  path=AnimGraph.AnimGraphNode_StateMachine_2.Locomotion.AnimStateNode_1.Moving.AnimGraphNode_StateMachine_0.JumpLand.AnimStateTransitionNode_1.Transition
  nodes [Result()]
 }
 JumpStart {
  path=AnimGraph.AnimGraphNode_StateMachine_2.Locomotion.AnimStateNode_2.JumpStart
  nodes [Output Animation Pose(Result=Sequence Player 'Grandma_Jump_Start'() {bLoopAnimation=false, sequence=&/Grandma_Jump_Start})]
 }
 FallAsleep {
  path=AnimGraph.AnimGraphNode_StateMachine_2.Locomotion.AnimStateNode_3.FallAsleep
  nodes [Output Animation Pose(Result=Sequence Player 'Grandma_EnergyDep_Intro'() {bLoopAnimation=false, sequence=/Irwin/AI/Predators/Grandma/Animations/EnergyDepleted/Grandma_EnergyDep_Intro})]
 }
 Sleeping {
  path=AnimGraph.AnimGraphNode_StateMachine_2.Locomotion.AnimStateNode_4.Sleeping
  nodes [Output Animation Pose(Result=Sequence Player 'Grandma_EnergyDep_Loop'() {sequence=/Irwin/AI/Predators/Grandma/Animations/EnergyDepleted/Grandma_EnergyDep_Loop})]
 }
 Fall {
  path=AnimGraph.AnimGraphNode_StateMachine_2.Locomotion.AnimStateNode_5.Fall
  nodes [Output Animation Pose(Result=Sequence Player 'Grandma_Jump_Fall_Loop'() {playRate=2, sequence=&/Grandma_Jump_Fall_Loop})]
 }
 JumpLand {
  path=AnimGraph.AnimGraphNode_StateMachine_2.Locomotion.AnimStateNode_6.JumpLand
  nodes [Output Animation Pose(Result=Sequence Player 'Grandma_Jump_Land'() {bLoopAnimation=false, sequence=&/Grandma_Jump_Land})]
 }
 WakeUp {
  path=AnimGraph.AnimGraphNode_StateMachine_2.Locomotion.AnimStateNode_7.WakeUp
  nodes [Output Animation Pose(Result=Sequence Player 'Grandma_EnergyDep_Outro'() {bLoopAnimation=false, sequence=/Irwin/AI/Predators/Grandma/Animations/EnergyDepleted/Grandma_EnergyDep_Outro})]
 }
 Swimming {
  path=AnimGraph.AnimGraphNode_StateMachine_2.Locomotion.AnimStateNode_9.Swimming
  nodes [
   n1 = Sequence Player 'Grandma_Swim_Idle'() {sequence=&/Grandma_Swim_Idle}
   n2 = Sequence Player 'Grandma_Swim_N'() {sequence=&/Grandma_Swim_N}
   n3 = Get bStateRule_SwimIdle_SwimLoco() -> (bStateRule_SwimIdle_SwimLoco:bool)
   n4 = BreakRotator(InRot=Get PawnRotationRate())
   Output Animation Pose(Result=Apply Additive(
     Base=Blend Poses by bool(BlendPose_0=@n2, BlendPose_1=@n1, BlendTime_0=0.5, BlendTime_1=1, bActiveValue=@n3) {blendType=Linear, transitionType=Inertialization}
     Additive=Blendspace Player 'Grandma_Walk_N_BS_TurnPose'(X=@n4.Yaw) {blendSpace=&Grandma_Walk_N_BS_TurnPose}
     bAlphaBoolEnabled=Get bLookAtAdditiveTurnAlpha()
    ) {alphaInputType=Bool, alphaBoolBlend {blendInTime=0.35, blendOption=Linear, blendOutTime=0.35}})
  ]
 }
 Transition {
  path=AnimGraph.AnimGraphNode_StateMachine_2.Locomotion.AnimStateTransitionNode_0.Transition
  nodes [Result(bCanEnterTransition=Get bStateRule_Fall_Moving() -> (bStateRule_Fall_Moving:bool))]
 }
 Transition {
  path=AnimGraph.AnimGraphNode_StateMachine_2.Locomotion.AnimStateTransitionNode_1.Transition
  nodes [Result(bCanEnterTransition=Get bIsSurfaceSwimming())]
 }
 Transition {
  path=AnimGraph.AnimGraphNode_StateMachine_2.Locomotion.AnimStateTransitionNode_10.Transition
  nodes [Result(bCanEnterTransition=Get StateTransition_SleepingToWakeUp() -> (StateTransition_SleepingToWakeUp:bool))]
 }
 Transition {
  path=AnimGraph.AnimGraphNode_StateMachine_2.Locomotion.AnimStateTransitionNode_11.Transition
  nodes [Result(bCanEnterTransition=Get StateTransition_WakeUpToIdle() -> (StateTransition_WakeUpToIdle:bool))]
 }
 Transition {
  path=AnimGraph.AnimGraphNode_StateMachine_2.Locomotion.AnimStateTransitionNode_13.Transition
  nodes [Result(bCanEnterTransition=Get bStateRule_Moving_Idle() -> (bStateRule_Moving_Idle:bool))]
 }
 Transition {
  path=AnimGraph.AnimGraphNode_StateMachine_2.Locomotion.AnimStateTransitionNode_14.Transition
  nodes [Result(bCanEnterTransition=Get bStateRule_JumpLand_Idle() -> (bStateRule_JumpLand_Idle:bool))]
 }
 Transition {
  path=AnimGraph.AnimGraphNode_StateMachine_2.Locomotion.AnimStateTransitionNode_2.Transition
  nodes [Result(bCanEnterTransition=Get bIsSurfaceSwimming())]
 }
 Transition {
  path=AnimGraph.AnimGraphNode_StateMachine_2.Locomotion.AnimStateTransitionNode_29.Transition
  nodes [Result(bCanEnterTransition=Get bIsSurfaceSwimming())]
 }
 Transition {
  path=AnimGraph.AnimGraphNode_StateMachine_2.Locomotion.AnimStateTransitionNode_3.Transition
  nodes [Result(bCanEnterTransition=Get StateTransition_IdleToFallAsleep() -> (StateTransition_IdleToFallAsleep:bool))]
 }
 Transition {
  path=AnimGraph.AnimGraphNode_StateMachine_2.Locomotion.AnimStateTransitionNode_30.Transition
  nodes [Result(bCanEnterTransition=NOT Boolean(A=Get bIsSurfaceSwimming()))]
 }
 AnimationTransitionGraph_0 {
  path=AnimGraph.AnimGraphNode_StateMachine_2.Locomotion.AnimStateTransitionNode_37.AnimationTransitionGraph_0
  nodes [Result(bCanEnterTransition=Get bStateRule_Idle_Moving() -> (bStateRule_Idle_Moving:bool))]
 }
 AnimationTransitionGraph_0 {
  path=AnimGraph.AnimGraphNode_StateMachine_2.Locomotion.AnimStateTransitionNode_38.AnimationTransitionGraph_0
  nodes [Result(bCanEnterTransition=Get bIsJumping())]
 }
 Transition {path=AnimGraph.AnimGraphNode_StateMachine_2.Locomotion.AnimStateTransitionNode_4.Transition, nodes [Result()]}
 Transition {
  path=AnimGraph.AnimGraphNode_StateMachine_2.Locomotion.AnimStateTransitionNode_40.Transition
  nodes [Result(bCanEnterTransition=Get bIsSurfaceSwimming())]
 }
 AnimationTransitionGraph_0 {
  path=AnimGraph.AnimGraphNode_StateMachine_2.Locomotion.AnimStateTransitionNode_41.AnimationTransitionGraph_0
  nodes [Result(bCanEnterTransition=Get bIsJumping())]
 }
 Transition {
  path=AnimGraph.AnimGraphNode_StateMachine_2.Locomotion.AnimStateTransitionNode_5.Transition
  nodes [Result(bCanEnterTransition=Get StateTransition_FallAsleepToSleeping() -> (StateTransition_FallAsleepToSleeping:bool))]
 }
 AnimationTransitionGraph_0 {
  path=AnimGraph.AnimGraphNode_StateMachine_2.Locomotion.AnimStateTransitionNode_6.AnimationTransitionGraph_0
  nodes [Result(bCanEnterTransition=Get bIsJumping())]
 }
 Transition {
  path=AnimGraph.AnimGraphNode_StateMachine_2.Locomotion.AnimStateTransitionNode_7.Transition
  nodes [Result(bCanEnterTransition=Get bStateRule_JumpLand_Moving() -> (bStateRule_JumpLand_Moving:bool))]
 }
 AnimationTransitionGraph_0 {
  path=AnimGraph.AnimGraphNode_StateMachine_2.Locomotion.AnimStateTransitionNode_8.AnimationTransitionGraph_0
  nodes [Result(bCanEnterTransition=Get bStateRule_Fall_Idle() -> (bStateRule_Fall_Idle:bool))]
 }
 Transition {
  path=AnimGraph.AnimGraphNode_StateMachine_2.Locomotion.AnimStateTransitionNode_9.Transition
  nodes [Result(bCanEnterTransition=Get bIsFalling())]
 }
 RidingLocomotion {
  entry=Idle
  path=AnimGraph.AnimGraphNode_StateMachine_3.RidingLocomotion
  ids=id
  nodes [
   SA_ToJumpStart() {is=title AnimStateAliasNode, also {aliases [Idle,JumpLand,Moving,Stop,TurnInPlace,TurnSettle]}}
   SA_ToFall() {is=title AnimStateAliasNode, also {aliases [Moving,Stop,TurnInPlace,TurnSettle]}}
   SA_ToSwimming() {is=title AnimStateAliasNode, bGlobalAlias=true}
   SA_ToMoving() {is=title AnimStateAliasNode, also {aliases [Idle,JumpLand,TurnInPlace,TurnSettle]}}
   AnimStateNode_0 = Idle() {
    stateEntered {guid=#2}
    stateFullyBlended {guid=#3}
    stateLeft {guid=#4}
   }
   AnimStateNode_1 = Moving() {
    stateEntered {guid=#5}
    stateFullyBlended {guid=#6}
    stateLeft {guid=#7}
   }
   AnimStateNode_2 = JumpStart()
   AnimStateNode_3 = Stop() {
    is=title AnimStateNode
    stateEntered {guid=#10}
    stateFullyBlended {guid=#11}
    stateLeft {guid=#12}
   }
   AnimStateNode_4 = TurnInPlace() {
    is=title AnimStateNode
    stateEntered {guid=#13}
    stateFullyBlended {guid=#14}
    stateLeft {guid=#15}
   }
   AnimStateNode_5 = Fall() {stateLeft {guid=#8, notifyName=FallExit}}
   AnimStateNode_6 = JumpLand() {is=title AnimStateNode, stateEntered {guid=#9, notifyName=LandEnter}}
   AnimStateNode_7 = TurnSettle() {is=title AnimStateNode}
   AnimStateNode_9 = Swimming()
   AnimStateTransitionNode_0 = AnimStateTransitionNode() {crossfadeDuration=0.15, logicType=TLT_Inertialization}
   AnimStateTransitionNode_1 = AnimStateTransitionNode() {logicType=TLT_Inertialization}
   AnimStateTransitionNode_10 = AnimStateTransitionNode()
   AnimStateTransitionNode_12 = AnimStateTransitionNode() {crossfadeDuration=0.4, logicType=TLT_Inertialization}
   AnimStateTransitionNode_13 = AnimStateTransitionNode() {crossfadeDuration=0.3, logicType=TLT_Inertialization, priorityOrder=100}
   AnimStateTransitionNode_14 = AnimStateTransitionNode() {crossfadeDuration=0.3, logicType=TLT_Inertialization}
   AnimStateTransitionNode_15 = AnimStateTransitionNode() {crossfadeDuration=0.4, logicType=TLT_Inertialization}
   AnimStateTransitionNode_16 = AnimStateTransitionNode() {crossfadeDuration=0.6, logicType=TLT_Inertialization, priorityOrder=2}
   AnimStateTransitionNode_18 = AnimStateTransitionNode() {logicType=TLT_Inertialization}
   AnimStateTransitionNode_2 = AnimStateTransitionNode() {crossfadeDuration=0.35, logicType=TLT_Inertialization, priorityOrder=3}
   AnimStateTransitionNode_3 = AnimStateTransitionNode() {bAutomaticRuleBasedOnSequencePlayerInState=true, crossfadeDuration=0.4, logicType=TLT_Inertialization, priorityOrder=3}
   AnimStateTransitionNode_39 = AnimStateTransitionNode() {bAutomaticRuleBasedOnSequencePlayerInState=true, crossfadeDuration=0.3, priorityOrder=100}
   AnimStateTransitionNode_4 = AnimStateTransitionNode() {bAutomaticRuleBasedOnSequencePlayerInState=true, crossfadeDuration=0.4, logicType=TLT_Inertialization}
   AnimStateTransitionNode_5 = AnimStateTransitionNode() {bAutomaticRuleBasedOnSequencePlayerInState=true, crossfadeDuration=0.3, priorityOrder=2}
   AnimStateTransitionNode_6 = AnimStateTransitionNode()
   AnimStateTransitionNode_7 = AnimStateTransitionNode() {logicType=TLT_Inertialization, priorityOrder=2}
   AnimStateTransitionNode_8 = AnimStateTransitionNode() {logicType=TLT_Inertialization, priorityOrder=100}
   AnimStateTransitionNode_9 = AnimStateTransitionNode() {blendMode=Linear, crossfadeDuration=0.3, priorityOrder=2}
  ]
  links [
   {
    dst=JumpLand
    rule=AnimGraph.AnimGraphNode_StateMachine_3.RidingLocomotion.AnimStateTransitionNode_8.AnimationTransitionGraph_0
    src=Fall
    via=AnimStateTransitionNode_8
   }
   {
    dst=Moving
    rule=AnimGraph.AnimGraphNode_StateMachine_3.RidingLocomotion.AnimStateTransitionNode_0.Transition
    src=Fall
    via=AnimStateTransitionNode_0
   }
   {
    dst=TurnInPlace
    rule=AnimGraph.AnimGraphNode_StateMachine_3.RidingLocomotion.AnimStateTransitionNode_10.AnimationTransitionGraph_1
    src=Idle
    via=AnimStateTransitionNode_10
   }
   {
    dst=Idle
    rule=AnimGraph.AnimGraphNode_StateMachine_3.RidingLocomotion.AnimStateTransitionNode_39.Transition
    src=JumpLand
    via=AnimStateTransitionNode_39
   }
   {
    dst=Moving
    rule=AnimGraph.AnimGraphNode_StateMachine_3.RidingLocomotion.AnimStateTransitionNode_7.Transition
    src=JumpLand
    via=AnimStateTransitionNode_7
   }
   {
    dst=Fall
    rule=AnimGraph.AnimGraphNode_StateMachine_3.RidingLocomotion.AnimStateTransitionNode_4.Transition
    src=JumpStart
    via=AnimStateTransitionNode_4
   }
   {
    dst=Stop
    rule=AnimGraph.AnimGraphNode_StateMachine_3.RidingLocomotion.AnimStateTransitionNode_2.Transition
    src=Moving
    via=AnimStateTransitionNode_2
   }
   {
    dst=Fall
    rule=AnimGraph.AnimGraphNode_StateMachine_3.RidingLocomotion.AnimStateTransitionNode_14.Transition
    src=SA_ToFall
    via=AnimStateTransitionNode_14
   }
   {
    dst=JumpStart
    rule=AnimGraph.AnimGraphNode_StateMachine_3.RidingLocomotion.AnimStateTransitionNode_13.Transition
    src=SA_ToJumpStart
    via=AnimStateTransitionNode_13
   }
   {
    dst=Moving
    rule=AnimGraph.AnimGraphNode_StateMachine_3.RidingLocomotion.AnimStateTransitionNode_16.Transition
    src=SA_ToMoving
    via=AnimStateTransitionNode_16
   }
   {
    dst=Swimming
    rule=AnimGraph.AnimGraphNode_StateMachine_3.RidingLocomotion.AnimStateTransitionNode_15.Transition
    src=SA_ToSwimming
    via=AnimStateTransitionNode_15
   }
   {
    dst=Idle
    rule=AnimGraph.AnimGraphNode_StateMachine_3.RidingLocomotion.AnimStateTransitionNode_3.Transition
    src=Stop
    via=AnimStateTransitionNode_3
   }
   {
    dst=Moving
    rule=AnimGraph.AnimGraphNode_StateMachine_3.RidingLocomotion.AnimStateTransitionNode_9.Transition
    src=Stop
    via=AnimStateTransitionNode_9
   }
   {
    dst=TurnInPlace
    rule=AnimGraph.AnimGraphNode_StateMachine_3.RidingLocomotion.AnimStateTransitionNode_12.Transition
    src=Stop
    via=AnimStateTransitionNode_12
   }
   {
    dst=Moving
    rule=AnimGraph.AnimGraphNode_StateMachine_3.RidingLocomotion.AnimStateTransitionNode_18.Transition
    src=Swimming
    via=AnimStateTransitionNode_18
   }
   {
    dst=TurnSettle
    rule=AnimGraph.AnimGraphNode_StateMachine_3.RidingLocomotion.AnimStateTransitionNode_1.Transition
    src=TurnInPlace
    via=AnimStateTransitionNode_1
   }
   {
    dst=Idle
    rule=AnimGraph.AnimGraphNode_StateMachine_3.RidingLocomotion.AnimStateTransitionNode_5.Transition
    src=TurnSettle
    via=AnimStateTransitionNode_5
   }
   {
    dst=TurnInPlace
    rule=AnimGraph.AnimGraphNode_StateMachine_3.RidingLocomotion.AnimStateTransitionNode_6.Transition
    src=TurnSettle
    via=AnimStateTransitionNode_6
   }
  ]
 }
 Idle {
  path=AnimGraph.AnimGraphNode_StateMachine_3.RidingLocomotion.AnimStateNode_0.Idle
  nodes [Output Animation Pose(Result=Sequence Player 'Grandma_Idle_Riding'() {sequence=&/Grandma_Idle_Riding})]
 }
 Moving {
  path=AnimGraph.AnimGraphNode_StateMachine_3.RidingLocomotion.AnimStateNode_1.Moving
  nodes [
   n1 = Break Riding Animation Data(RidingAnimationData=Get RidingAnimationData()) -> (
    bIsRidden=~
    bIsTurningInPlace=~
    bIsSprinting=~
    bIsBeingPetted=~
    bIsMovingBackwards=~
    TurnInPlaceRotationSpeed=~
    TurnInPlaceAngleDelta=~
    RiderReferentialYaw=~
   )
   n2 = Apply Additive!(
    Base=Sequence Player 'Grandma_Walk_S'!(PlayRate:float=Get BackwardsPlayRate() -> (BackwardsPlayRate:float)) -> (Pose) {groupName=Locomotion, method=SyncGroup, sequence=&/Grandma_Walk_S}
    Additive=Blendspace Player 'Grandma_Walk_S_Add_BS'!(X=@n1.RiderReferentialYaw) -> (Pose) {blendSpace=&/Grandma_Walk_S_Add_BS}
    Alpha:float
   ) -> (Pose) {alphaBoolBlend {blendInTime=0.5, blendOption=Linear, blendOutTime=0.35}}
   n3 = Blendspace Player 'Grandma_WalkRun_BS'!(X=Get PawnSpeed2D(), PlayRate:float=Get RidingMovePlayRate() -> (RidingMovePlayRate:float)) -> (Pose) {blendSpace=&Grandma_WalkRun_BS, groupName=Locomotion, method=SyncGroup}
   n4 = Sequence Player 'Grandma_Boost_N'!(PlayRate:float=Get BoostPlayRate() -> (BoostPlayRate:float)) -> (Pose) {groupName=Locomotion, method=SyncGroup, sequence=&/Grandma_Boost_N}
   n5 = BreakRotator(InRot=Get PawnRotationRate())
   n6 = Apply Additive(
    Base=Apply Additive!(
     Base=Blend Poses by bool(BlendPose_0=@n4, BlendPose_1=@n3, BlendTime_0=0.25, BlendTime_1=0.25, bActiveValue=@n1.bIsSprinting) {blendType=Linear}
     Additive=Blendspace Player 'Grandma_Walk_N_BS_TurnPose'(X=@n5.Yaw) {blendSpace=&Grandma_Walk_N_BS_TurnPose}
     bAlphaBoolEnabled:bool=Get bLookAtAdditiveTurnAlpha()
    ) -> (Pose) {alphaInputType=Bool, alphaBoolBlend {blendInTime=0.35, blendOption=Linear, blendOutTime=0.35}}
    Additive=Blendspace Player 'Grandma_BodyRoll_BS'(X=Get BodyRoll() -> (BodyRoll:float), Y=Get PawnSpeed2D()) {blendSpace=&/Grandma_BodyRoll_BS}
   ) {alphaBoolBlend {blendInTime=0.5, blendOption=Linear, blendOutTime=0.35}}
   AnimGraphNode_StateMachine_0 = JumpLand() {is=title AnimGraphNode_StateMachine}
   Output Animation Pose(Result=Apply Additive!(
     Base=Blend Poses by bool(BlendPose_0=@n2, BlendPose_1=@n6, BlendTime_0=0.3, BlendTime_1=0.3, bActiveValue=@n1.bIsMovingBackwards) {blendType=Linear}
     Additive=@AnimGraphNode_StateMachine_0
     Alpha:float
    ) -> (Pose) {alphaBoolBlend {blendOption=Linear}})
  ]
 }
 JumpLand {
  entry=AddRef
  path=AnimGraph.AnimGraphNode_StateMachine_3.RidingLocomotion.AnimStateNode_1.Moving.AnimGraphNode_StateMachine_0.JumpLand
  ids=id
  nodes [
   AnimStateNode_0 = AddRef()
   AnimStateNode_1 = LandAdd() {stateLeft {guid=#1, notifyName=LandAddExit}}
   AnimStateTransitionNode_0 = AnimStateTransitionNode() {crossfadeDuration=0.1, logicType=TLT_Inertialization}
   AnimStateTransitionNode_1 = AnimStateTransitionNode() {bAutomaticRuleBasedOnSequencePlayerInState=true, crossfadeDuration=0.3, logicType=TLT_Inertialization}
  ]
  links [
   {
    dst=LandAdd
    rule=AnimGraph.AnimGraphNode_StateMachine_3.RidingLocomotion.AnimStateNode_1.Moving.AnimGraphNode_StateMachine_0.JumpLand.AnimStateTransitionNode_0.Transition
    src=AddRef
    via=AnimStateTransitionNode_0
   }
   {
    dst=AddRef
    rule=AnimGraph.AnimGraphNode_StateMachine_3.RidingLocomotion.AnimStateNode_1.Moving.AnimGraphNode_StateMachine_0.JumpLand.AnimStateTransitionNode_1.Transition
    src=LandAdd
    via=AnimStateTransitionNode_1
   }
  ]
 }
 AddRef {
  path=AnimGraph.AnimGraphNode_StateMachine_3.RidingLocomotion.AnimStateNode_1.Moving.AnimGraphNode_StateMachine_0.JumpLand.AnimStateNode_0.AddRef
  nodes [Output Animation Pose(Result=Additive Identity Pose())]
 }
 LandAdd {
  path=AnimGraph.AnimGraphNode_StateMachine_3.RidingLocomotion.AnimStateNode_1.Moving.AnimGraphNode_StateMachine_0.JumpLand.AnimStateNode_1.LandAdd
  nodes [Output Animation Pose(Result="Sequence Player (additive) 'Grandma_JumpLand_Moving_Add'"() {sequence=&Grandma_JumpLand_Moving_Add})]
 }
 Transition {
  path=AnimGraph.AnimGraphNode_StateMachine_3.RidingLocomotion.AnimStateNode_1.Moving.AnimGraphNode_StateMachine_0.JumpLand.AnimStateTransitionNode_0.Transition
  nodes [Result(bCanEnterTransition=Get bPlayLandAdditive())]
 }
 Transition {
  path=AnimGraph.AnimGraphNode_StateMachine_3.RidingLocomotion.AnimStateNode_1.Moving.AnimGraphNode_StateMachine_0.JumpLand.AnimStateTransitionNode_1.Transition
  nodes [Result()]
 }
 JumpStart {
  path=AnimGraph.AnimGraphNode_StateMachine_3.RidingLocomotion.AnimStateNode_2.JumpStart
  nodes [Output Animation Pose(Result=Sequence Player 'Grandma_Jump_Start'() {bLoopAnimation=false, sequence=&Grandma_Jump_Start})]
 }
 Stop {
  path=AnimGraph.AnimGraphNode_StateMachine_3.RidingLocomotion.AnimStateNode_3.Stop
  nodes [
   AnimGraphNode_StateMachine_0 = StopTransition() {is=title AnimGraphNode_StateMachine, bAllowConduitEntryStates=true, bSkipFirstUpdateTransition=false, maxTransitionsPerFrame=1}
   Output Animation Pose(Result=@AnimGraphNode_StateMachine_0)
  ]
 }
 StopTransition {
  entry=Conduit
  path=AnimGraph.AnimGraphNode_StateMachine_3.RidingLocomotion.AnimStateNode_3.Stop.AnimGraphNode_StateMachine_0.StopTransition
  ids=id
  nodes [
   AnimStateConduitNode_0 = Conduit()
   AnimStateNode_0 = FeetInAir() {is=title AnimStateNode}
   AnimStateNode_1 = FrontFeetPlanted() {is=title AnimStateNode}
   AnimStateNode_2 = BackFeetPlanted() {is=title AnimStateNode}
   AnimStateNode_3 = Backwards() {is=title AnimStateNode, bAlwaysResetOnEntry=true}
   AnimStateTransitionNode_0 = AnimStateTransitionNode() {logicType=TLT_Inertialization}
   AnimStateTransitionNode_1 = AnimStateTransitionNode() {logicType=TLT_Inertialization, priorityOrder=2}
   AnimStateTransitionNode_2 = AnimStateTransitionNode() {logicType=TLT_Inertialization, priorityOrder=100}
   AnimStateTransitionNode_3 = AnimStateTransitionNode() {logicType=TLT_Inertialization, priorityOrder=0}
  ]
  links [
   {
    dst=BackFeetPlanted
    rule=AnimGraph.AnimGraphNode_StateMachine_3.RidingLocomotion.AnimStateNode_3.Stop.AnimGraphNode_StateMachine_0.StopTransition.AnimStateTransitionNode_2.Transition
    src=Conduit
    via=AnimStateTransitionNode_2
   }
   {
    dst=Backwards
    rule=AnimGraph.AnimGraphNode_StateMachine_3.RidingLocomotion.AnimStateNode_3.Stop.AnimGraphNode_StateMachine_0.StopTransition.AnimStateTransitionNode_3.Transition
    src=Conduit
    via=AnimStateTransitionNode_3
   }
   {
    dst=FeetInAir
    rule=AnimGraph.AnimGraphNode_StateMachine_3.RidingLocomotion.AnimStateNode_3.Stop.AnimGraphNode_StateMachine_0.StopTransition.AnimStateTransitionNode_0.Transition
    src=Conduit
    via=AnimStateTransitionNode_0
   }
   {
    dst=FrontFeetPlanted
    rule=AnimGraph.AnimGraphNode_StateMachine_3.RidingLocomotion.AnimStateNode_3.Stop.AnimGraphNode_StateMachine_0.StopTransition.AnimStateTransitionNode_1.Transition
    src=Conduit
    via=AnimStateTransitionNode_1
   }
  ]
 }
 Conduit {
  path=AnimGraph.AnimGraphNode_StateMachine_3.RidingLocomotion.AnimStateNode_3.Stop.AnimGraphNode_StateMachine_0.StopTransition.AnimStateConduitNode_0.Conduit
  nodes [Result(bCanEnterTransition=true)]
 }
 FeetInAir {
  path=AnimGraph.AnimGraphNode_StateMachine_3.RidingLocomotion.AnimStateNode_3.Stop.AnimGraphNode_StateMachine_0.StopTransition.AnimStateNode_0.FeetInAir
  nodes [Output Animation Pose(Result=Sequence Player 'Grandma_Stop_FeetInAir'() {bLoopAnimation=false, sequence=&/Grandma_Stop_FeetInAir})]
 }
 FrontFeetPlanted {
  path=AnimGraph.AnimGraphNode_StateMachine_3.RidingLocomotion.AnimStateNode_3.Stop.AnimGraphNode_StateMachine_0.StopTransition.AnimStateNode_1.FrontFeetPlanted
  nodes [Output Animation Pose(Result=Sequence Player 'Grandma_Stop_FrontFeetPlanted'() {bLoopAnimation=false, sequence=&/Grandma_Stop_FrontFeetPlanted})]
 }
 BackFeetPlanted {
  path=AnimGraph.AnimGraphNode_StateMachine_3.RidingLocomotion.AnimStateNode_3.Stop.AnimGraphNode_StateMachine_0.StopTransition.AnimStateNode_2.BackFeetPlanted
  nodes [Output Animation Pose(Result=Sequence Player 'Grandma_Stop_BackFeetPlanted'() {bLoopAnimation=false, sequence=&/Grandma_Stop_BackFeetPlanted})]
 }
 Backwards {
  path=AnimGraph.AnimGraphNode_StateMachine_3.RidingLocomotion.AnimStateNode_3.Stop.AnimGraphNode_StateMachine_0.StopTransition.AnimStateNode_3.Backwards
  nodes [
   AnimGraphNode_StateMachine_0 = StopBackwards() {is=title AnimGraphNode_StateMachine, bAllowConduitEntryStates=true, bSkipFirstUpdateTransition=false, maxTransitionsPerFrame=1}
   Output Animation Pose(Result=@AnimGraphNode_StateMachine_0)
  ]
 }
 StopBackwards {
  entry=Conduit
  path=AnimGraph.AnimGraphNode_StateMachine_3.RidingLocomotion.AnimStateNode_3.Stop.AnimGraphNode_StateMachine_0.StopTransition.AnimStateNode_3.Backwards.AnimGraphNode_StateMachine_0.StopBackwards
  ids=id
  nodes [
   AnimStateConduitNode_0 = Conduit()
   AnimStateNode_0 = RightPlantedLeftPass() {is=title AnimStateNode}
   AnimStateNode_1 = RightBackFeetForward() {is=title AnimStateNode}
   AnimStateNode_2 = LeftPlantedRightPass() {is=title AnimStateNode}
   AnimStateNode_3 = LeftBackFeetForward() {is=title AnimStateNode}
   AnimStateTransitionNode_0 = AnimStateTransitionNode() {logicType=TLT_Inertialization, priorityOrder=100}
   AnimStateTransitionNode_1 = AnimStateTransitionNode() {logicType=TLT_Inertialization, priorityOrder=2}
   AnimStateTransitionNode_2 = AnimStateTransitionNode() {logicType=TLT_Inertialization, priorityOrder=3}
   AnimStateTransitionNode_3 = AnimStateTransitionNode() {logicType=TLT_Inertialization}
  ]
  links [
   {
    dst=LeftBackFeetForward
    rule=AnimGraph.AnimGraphNode_StateMachine_3.RidingLocomotion.AnimStateNode_3.Stop.AnimGraphNode_StateMachine_0.StopTransition.AnimStateNode_3.Backwards.AnimGraphNode_StateMachine_0.StopBackwards.AnimStateTransitionNode_3.Transition
    src=Conduit
    via=AnimStateTransitionNode_3
   }
   {
    dst=LeftPlantedRightPass
    rule=AnimGraph.AnimGraphNode_StateMachine_3.RidingLocomotion.AnimStateNode_3.Stop.AnimGraphNode_StateMachine_0.StopTransition.AnimStateNode_3.Backwards.AnimGraphNode_StateMachine_0.StopBackwards.AnimStateTransitionNode_2.Transition
    src=Conduit
    via=AnimStateTransitionNode_2
   }
   {
    dst=RightBackFeetForward
    rule=AnimGraph.AnimGraphNode_StateMachine_3.RidingLocomotion.AnimStateNode_3.Stop.AnimGraphNode_StateMachine_0.StopTransition.AnimStateNode_3.Backwards.AnimGraphNode_StateMachine_0.StopBackwards.AnimStateTransitionNode_1.Transition
    src=Conduit
    via=AnimStateTransitionNode_1
   }
   {
    dst=RightPlantedLeftPass
    rule=AnimGraph.AnimGraphNode_StateMachine_3.RidingLocomotion.AnimStateNode_3.Stop.AnimGraphNode_StateMachine_0.StopTransition.AnimStateNode_3.Backwards.AnimGraphNode_StateMachine_0.StopBackwards.AnimStateTransitionNode_0.Transition
    src=Conduit
    via=AnimStateTransitionNode_0
   }
  ]
 }
 Conduit {
  path=AnimGraph.AnimGraphNode_StateMachine_3.RidingLocomotion.AnimStateNode_3.Stop.AnimGraphNode_StateMachine_0.StopTransition.AnimStateNode_3.Backwards.AnimGraphNode_StateMachine_0.StopBackwards.AnimStateConduitNode_0.Conduit
  nodes [Result(bCanEnterTransition=true)]
 }
 RightPlantedLeftPass {
  path=AnimGraph.AnimGraphNode_StateMachine_3.RidingLocomotion.AnimStateNode_3.Stop.AnimGraphNode_StateMachine_0.StopTransition.AnimStateNode_3.Backwards.AnimGraphNode_StateMachine_0.StopBackwards.AnimStateNode_0.RightPlantedLeftPass
  nodes [Output Animation Pose(Result=Sequence Player 'Grandma_Stop_Backward_fr7'() {bLoopAnimation=false, sequence=&/Grandma_Stop_Backward_fr7})]
 }
 RightBackFeetForward {
  path=AnimGraph.AnimGraphNode_StateMachine_3.RidingLocomotion.AnimStateNode_3.Stop.AnimGraphNode_StateMachine_0.StopTransition.AnimStateNode_3.Backwards.AnimGraphNode_StateMachine_0.StopBackwards.AnimStateNode_1.RightBackFeetForward
  nodes [Output Animation Pose(Result=Sequence Player 'Grandma_Stop_Backward_fr19'() {bLoopAnimation=false, sequence=&/Grandma_Stop_Backward_fr19})]
 }
 LeftPlantedRightPass {
  path=AnimGraph.AnimGraphNode_StateMachine_3.RidingLocomotion.AnimStateNode_3.Stop.AnimGraphNode_StateMachine_0.StopTransition.AnimStateNode_3.Backwards.AnimGraphNode_StateMachine_0.StopBackwards.AnimStateNode_2.LeftPlantedRightPass
  nodes [Output Animation Pose(Result=Sequence Player 'Grandma_Stop_Backward_fr24'() {bLoopAnimation=false, sequence=&/Grandma_Stop_Backward_fr24})]
 }
 LeftBackFeetForward {
  path=AnimGraph.AnimGraphNode_StateMachine_3.RidingLocomotion.AnimStateNode_3.Stop.AnimGraphNode_StateMachine_0.StopTransition.AnimStateNode_3.Backwards.AnimGraphNode_StateMachine_0.StopBackwards.AnimStateNode_3.LeftBackFeetForward
  nodes [Output Animation Pose(Result=Sequence Player 'Grandma_Stop_Backward'() {bLoopAnimation=false, sequence=&/Grandma_Stop_Backward})]
 }
 Transition {
  path=AnimGraph.AnimGraphNode_StateMachine_3.RidingLocomotion.AnimStateNode_3.Stop.AnimGraphNode_StateMachine_0.StopTransition.AnimStateNode_3.Backwards.AnimGraphNode_StateMachine_0.StopBackwards.AnimStateTransitionNode_0.Transition
  nodes [Result(bCanEnterTransition=true)]
 }
 Transition {
  path=AnimGraph.AnimGraphNode_StateMachine_3.RidingLocomotion.AnimStateNode_3.Stop.AnimGraphNode_StateMachine_0.StopTransition.AnimStateNode_3.Backwards.AnimGraphNode_StateMachine_0.StopBackwards.AnimStateTransitionNode_1.Transition
  nodes [Result()]
 }
 Transition {
  path=AnimGraph.AnimGraphNode_StateMachine_3.RidingLocomotion.AnimStateNode_3.Stop.AnimGraphNode_StateMachine_0.StopTransition.AnimStateNode_3.Backwards.AnimGraphNode_StateMachine_0.StopBackwards.AnimStateTransitionNode_2.Transition
  nodes [Result()]
 }
 Transition {
  path=AnimGraph.AnimGraphNode_StateMachine_3.RidingLocomotion.AnimStateNode_3.Stop.AnimGraphNode_StateMachine_0.StopTransition.AnimStateNode_3.Backwards.AnimGraphNode_StateMachine_0.StopBackwards.AnimStateTransitionNode_3.Transition
  nodes [Result()]
 }
 Transition {
  path=AnimGraph.AnimGraphNode_StateMachine_3.RidingLocomotion.AnimStateNode_3.Stop.AnimGraphNode_StateMachine_0.StopTransition.AnimStateTransitionNode_0.Transition
  nodes [Result()]
 }
 Transition {
  path=AnimGraph.AnimGraphNode_StateMachine_3.RidingLocomotion.AnimStateNode_3.Stop.AnimGraphNode_StateMachine_0.StopTransition.AnimStateTransitionNode_1.Transition
  nodes [Result()]
 }
 Transition {
  path=AnimGraph.AnimGraphNode_StateMachine_3.RidingLocomotion.AnimStateNode_3.Stop.AnimGraphNode_StateMachine_0.StopTransition.AnimStateTransitionNode_2.Transition
  nodes [Result(bCanEnterTransition=true)]
 }
 Transition {
  path=AnimGraph.AnimGraphNode_StateMachine_3.RidingLocomotion.AnimStateNode_3.Stop.AnimGraphNode_StateMachine_0.StopTransition.AnimStateTransitionNode_3.Transition
  nodes [Result(bCanEnterTransition=Get bShouldPlayStopBackwards() -> (bShouldPlayStopBackwards:bool))]
 }
 TurnInPlace {
  path=AnimGraph.AnimGraphNode_StateMachine_3.RidingLocomotion.AnimStateNode_4.TurnInPlace
  nodes [Output Animation Pose(Result=Blend Poses by bool(
     BlendPose_0=Sequence Player 'Grandma_Riding_TurnRight_90'!(PlayRate:float=1) -> (Pose) {bLoopAnimation=false, sequence=&/Grandma_Riding_TurnRight_90}
     BlendPose_1=Sequence Player 'Grandma_Riding_TurnLeft_90'!(PlayRate:float=1) -> (Pose) {bLoopAnimation=false, sequence=&/Grandma_Riding_TurnLeft_90}
    ) {blendType=Linear})]
 }
 Fall {
  path=AnimGraph.AnimGraphNode_StateMachine_3.RidingLocomotion.AnimStateNode_5.Fall
  nodes [
   n1 = Blendspace Player 'Grandma_SlopeSliding_BS'(X=Get SlopeSlidingRoll() -> (SlopeSlidingRoll:float), Y=Get SlopeSlidingPitch() -> (SlopeSlidingPitch:float)) {blendSpace=/Irwin/AI/Predators/Grandma/Animations/SlopeSliding/Grandma_SlopeSliding_BS}
   n2 = "Sequence Player (additive) 'Grandma_SlopeSliding_Additive'"() {sequence=/Irwin/AI/Predators/Grandma/Animations/SlopeSliding/Grandma_SlopeSliding_Additive}
   Output Animation Pose(Result=Blend Poses by bool(
     BlendPose_0=Apply Additive(Base=@n1, Additive=@n2) {lODThreshold=1, alphaBoolBlend {blendOption=Linear}}
     BlendPose_1=Sequence Player 'Grandma_Jump_Fall_Loop'() {playRate=2, sequence=&Grandma_Jump_Fall_Loop}
     bActiveValue=Get bIsSlopeSliding() -> (bIsSlopeSliding:bool)
    ) {blendType=Linear})
  ]
 }
 JumpLand {
  path=AnimGraph.AnimGraphNode_StateMachine_3.RidingLocomotion.AnimStateNode_6.JumpLand
  nodes [Output Animation Pose(Result=Sequence Player 'Grandma_Jump_Land'() {bLoopAnimation=false, sequence=&Grandma_Jump_Land})]
 }
 TurnSettle {
  path=AnimGraph.AnimGraphNode_StateMachine_3.RidingLocomotion.AnimStateNode_7.TurnSettle
  nodes [Output Animation Pose(Result=Blend Poses by bool(
     BlendPose_0=Sequence Player 'Grandma_Riding_TurnRight_90_Settle'() {bLoopAnimation=false, sequence=&/Grandma_Riding_TurnRight_90_Settle}
     BlendPose_1=Sequence Player 'Grandma_Riding_TurnLeft_90_Settle'() {bLoopAnimation=false, sequence=&/Grandma_Riding_TurnLeft_90_Settle}
     BlendTime_0=0.25
     BlendTime_1=0.25
     bActiveValue=Get bWasTurningRight() -> (bWasTurningRight:bool)
    ) {blendType=Linear})]
 }
 Swimming {
  path=AnimGraph.AnimGraphNode_StateMachine_3.RidingLocomotion.AnimStateNode_9.Swimming
  nodes [
   n1 = Blend Poses by bool(
    BlendPose_0=Sequence Player 'Grandma_Swim_S'() {groupName=Locomotion, method=SyncGroup, sequence=&/Grandma_Swim_S}
    BlendPose_1=Sequence Player 'Grandma_Swim_N'() {groupName=Locomotion, method=SyncGroup, sequence=&Grandma_Swim_N}
    BlendTime_0=0.35
    BlendTime_1=0.35
    bActiveValue=Get bIsMovingBackwards() -> (bIsMovingBackwards:bool)
   ) {blendType=Linear, transitionType=Inertialization}
   n2 = Sequence Player 'Grandma_Swim_Idle'() {sequence=&Grandma_Swim_Idle}
   n3 = Get bStateRule_SwimIdle_SwimLoco()
   n4 = Get SwimmingYaw() -> (SwimmingYaw:float)
   Output Animation Pose(Result=Apply Additive!(
     Base=Blend Poses by bool(BlendPose_0=@n1, BlendPose_1=@n2, BlendTime_0=0.5, BlendTime_1=1, bActiveValue=@n3) {blendType=Linear}
     Additive=Blendspace Player 'Grandma_Walk_N_BS_TurnPose'(X=@n4) {blendSpace=&Grandma_Walk_N_BS_TurnPose}
     bAlphaBoolEnabled:bool=Get bLookAtAdditiveTurnAlpha()
    ) -> (Pose) {alphaInputType=Bool, alphaBoolBlend {blendInTime=0.35, blendOption=Linear, blendOutTime=0.35}})
  ]
 }
 Transition {
  path=AnimGraph.AnimGraphNode_StateMachine_3.RidingLocomotion.AnimStateTransitionNode_0.Transition
  nodes [Result(bCanEnterTransition=Get bStateRule_Fall_Moving())]
 }
 Transition {
  path=AnimGraph.AnimGraphNode_StateMachine_3.RidingLocomotion.AnimStateTransitionNode_1.Transition
  nodes [
   n1 = Get RidingAnimationData()
   Result(bCanEnterTransition=NOT Boolean(A=Break Riding Animation Data(RidingAnimationData=@n1) -> (
      bIsRidden=~
      bIsTurningInPlace=~
      bIsSprinting=~
      bIsBeingPetted=~
      bIsMovingBackwards=~
      TurnInPlaceRotationSpeed=~
      TurnInPlaceAngleDelta=~
      RiderReferentialYaw=~
     ).bIsTurningInPlace))
  ]
 }
 AnimationTransitionGraph_1 {
  path=AnimGraph.AnimGraphNode_StateMachine_3.RidingLocomotion.AnimStateTransitionNode_10.AnimationTransitionGraph_1
  nodes [Result(bCanEnterTransition=Break Riding Animation Data(RidingAnimationData=Get RidingAnimationData()) -> (
     bIsRidden=~
     bIsTurningInPlace=~
     bIsSprinting=~
     bIsBeingPetted=~
     bIsMovingBackwards=~
     TurnInPlaceRotationSpeed=~
     TurnInPlaceAngleDelta=~
     RiderReferentialYaw=~
    ).bIsTurningInPlace)]
 }
 Transition {
  path=AnimGraph.AnimGraphNode_StateMachine_3.RidingLocomotion.AnimStateTransitionNode_12.Transition
  nodes [Result(bCanEnterTransition=Break Riding Animation Data(RidingAnimationData=Get RidingAnimationData()) -> (
     bIsRidden=~
     bIsTurningInPlace=~
     bIsSprinting=~
     bIsBeingPetted=~
     bIsMovingBackwards=~
     TurnInPlaceRotationSpeed=~
     TurnInPlaceAngleDelta=~
     RiderReferentialYaw=~
    ).bIsTurningInPlace)]
 }
 Transition {
  path=AnimGraph.AnimGraphNode_StateMachine_3.RidingLocomotion.AnimStateTransitionNode_13.Transition
  nodes [Result(bCanEnterTransition=Get bIsJumping())]
 }
 Transition {
  path=AnimGraph.AnimGraphNode_StateMachine_3.RidingLocomotion.AnimStateTransitionNode_14.Transition
  nodes [Result(bCanEnterTransition=Get bIsFalling())]
 }
 Transition {
  path=AnimGraph.AnimGraphNode_StateMachine_3.RidingLocomotion.AnimStateTransitionNode_15.Transition
  nodes [Result(bCanEnterTransition=Get bIsSurfaceSwimming())]
 }
 Transition {
  path=AnimGraph.AnimGraphNode_StateMachine_3.RidingLocomotion.AnimStateTransitionNode_16.Transition
  nodes [Result(bCanEnterTransition=Get bStateRule_Idle_Moving())]
 }
 Transition {
  path=AnimGraph.AnimGraphNode_StateMachine_3.RidingLocomotion.AnimStateTransitionNode_18.Transition
  nodes [Result(bCanEnterTransition=NOT Boolean(A=Get bIsSurfaceSwimming()))]
 }
 Transition {
  path=AnimGraph.AnimGraphNode_StateMachine_3.RidingLocomotion.AnimStateTransitionNode_2.Transition
  nodes [Result(bCanEnterTransition=Get bStateRule_Moving_Idle())]
 }
 Transition {path=AnimGraph.AnimGraphNode_StateMachine_3.RidingLocomotion.AnimStateTransitionNode_3.Transition, nodes [Result()]}
 Transition {path=AnimGraph.AnimGraphNode_StateMachine_3.RidingLocomotion.AnimStateTransitionNode_39.Transition, nodes [Result()]}
 Transition {path=AnimGraph.AnimGraphNode_StateMachine_3.RidingLocomotion.AnimStateTransitionNode_4.Transition, nodes [Result()]}
 Transition {path=AnimGraph.AnimGraphNode_StateMachine_3.RidingLocomotion.AnimStateTransitionNode_5.Transition, nodes [Result()]}
 Transition {
  path=AnimGraph.AnimGraphNode_StateMachine_3.RidingLocomotion.AnimStateTransitionNode_6.Transition
  nodes [Result(bCanEnterTransition=Break Riding Animation Data(RidingAnimationData=Get RidingAnimationData()) -> (
     bIsRidden=~
     bIsTurningInPlace=~
     bIsSprinting=~
     bIsBeingPetted=~
     bIsMovingBackwards=~
     TurnInPlaceRotationSpeed=~
     TurnInPlaceAngleDelta=~
     RiderReferentialYaw=~
    ).bIsTurningInPlace)]
 }
 Transition {
  path=AnimGraph.AnimGraphNode_StateMachine_3.RidingLocomotion.AnimStateTransitionNode_7.Transition
  nodes [Result(bCanEnterTransition=Get bStateRule_JumpLand_Moving())]
 }
 AnimationTransitionGraph_0 {
  path=AnimGraph.AnimGraphNode_StateMachine_3.RidingLocomotion.AnimStateTransitionNode_8.AnimationTransitionGraph_0
  nodes [Result(bCanEnterTransition=Get bStateRule_Fall_Idle())]
 }
 Transition {
  path=AnimGraph.AnimGraphNode_StateMachine_3.RidingLocomotion.AnimStateTransitionNode_9.Transition
  nodes [Result(bCanEnterTransition=Get bStateRule_Idle_Moving())]
 }
 EventGraph {nodes [
   AnimNotify_FallExit() {is=title K2Node_Event}
   > Set bPlayLandAdditive(bPlayLandAdditive:bool=true) -> (Output_Get:bool)
   AnimNotify_LandEnter() {is=title K2Node_Event}
   > Set bPlayLandAdditive(bPlayLandAdditive=false)
   AnimNotify_LandAddExit() {is=title K2Node_Event}
   > Set bPlayLandAdditive(bPlayLandAdditive=false)
  ]}
 Fauna Idle {nodes [
   Input Pose() {name=InPose Idle}
   Output Pose(Result=Sequence Player 'Grandma_Idle_Short'() {sequence=&/Grandma_Idle_Short}) {layerGroup=FaunaIdleOverride}
  ]}
 Fauna IdleTurn {nodes [
   Input Pose() {name=InPose IdleTurn}
   Output Pose(Result=Blend Poses by bool(
     BlendPose_0=Sequence Player 'Grandma_TurnLeft_90'() {sequence=&/Grandma_TurnLeft_90}
     BlendPose_1=Sequence Player 'Grandma_TurnRight_90'() {sequence=&/Grandma_TurnRight_90}
     BlendTime_0=0.35
     BlendTime_1=0.35
     bActiveValue=Get bAnimRule_NegativeTurnAngle() -> (bAnimRule_NegativeTurnAngle:bool)
    ) {blendType=Linear, childUpateMode=ResetChildOnActivate, transitionType=Inertialization}) {layerGroup=FaunaIdleOverride}
  ]}
 Fauna Jump {nodes [
   AnimGraphNode_StateMachine_0 = DefaultJump() {is=title AnimGraphNode_StateMachine}
   Output Pose(Result=@AnimGraphNode_StateMachine_0) {layerGroup=None}
  ]}
 DefaultJump {
  entry=JumpStart
  path=Fauna Jump.AnimGraphNode_StateMachine_0.DefaultJump
  ids=id
  nodes [
   AnimStateNode_2 = JumpStart()
   AnimStateNode_4 = JumpApex() {is=title AnimStateNode}
   AnimStateNode_5 = JumpFall() {is=title AnimStateNode}
   AnimStateNode_6 = JumpLand() {is=title AnimStateNode}
   AnimStateTransitionNode_0 = AnimStateTransitionNode() {crossfadeDuration=0, priorityOrder=0}
   AnimStateTransitionNode_1 = AnimStateTransitionNode()
   AnimStateTransitionNode_11 = AnimStateTransitionNode()
   AnimStateTransitionNode_12 = AnimStateTransitionNode()
   AnimStateTransitionNode_5 = AnimStateTransitionNode() {bAutomaticRuleBasedOnSequencePlayerInState=true}
   AnimStateTransitionNode_6 = AnimStateTransitionNode() {bAutomaticRuleBasedOnSequencePlayerInState=true}
   AnimStateTransitionNode_8 = AnimStateTransitionNode() {crossfadeDuration=0.033}
  ]
  links [
   {
    dst=JumpFall
    rule=Fauna Jump.AnimGraphNode_StateMachine_0.DefaultJump.AnimStateTransitionNode_12.AnimationTransitionGraph_1
    src=JumpApex
    via=AnimStateTransitionNode_12
   }
   {
    dst=JumpFall
    rule=Fauna Jump.AnimGraphNode_StateMachine_0.DefaultJump.AnimStateTransitionNode_6.AnimationTransitionGraph_1
    src=JumpApex
    via=AnimStateTransitionNode_6
   }
   {
    dst=JumpLand
    rule=Fauna Jump.AnimGraphNode_StateMachine_0.DefaultJump.AnimStateTransitionNode_8.AnimationTransitionGraph_1
    src=JumpFall
    via=AnimStateTransitionNode_8
   }
   {
    dst=JumpApex
    rule=Fauna Jump.AnimGraphNode_StateMachine_0.DefaultJump.AnimStateTransitionNode_1.AnimationTransitionGraph_0
    src=JumpLand
    via=AnimStateTransitionNode_1
   }
   {
    dst=JumpApex
    rule=Fauna Jump.AnimGraphNode_StateMachine_0.DefaultJump.AnimStateTransitionNode_0.AnimationTransitionGraph_0
    src=JumpStart
    via=AnimStateTransitionNode_0
   }
   {
    dst=JumpApex
    rule=Fauna Jump.AnimGraphNode_StateMachine_0.DefaultJump.AnimStateTransitionNode_11.AnimationTransitionGraph_1
    src=JumpStart
    via=AnimStateTransitionNode_11
   }
   {
    dst=JumpApex
    rule=Fauna Jump.AnimGraphNode_StateMachine_0.DefaultJump.AnimStateTransitionNode_5.AnimationTransitionGraph_1
    src=JumpStart
    via=AnimStateTransitionNode_5
   }
  ]
 }
 JumpStart {
  path=Fauna Jump.AnimGraphNode_StateMachine_0.DefaultJump.AnimStateNode_2.JumpStart
  nodes [Output Animation Pose(Result=Sequence Player 'Grandma_Jump_Start'() {bLoopAnimation=false, sequence=&Grandma_Jump_Start})]
 }
 JumpApex {
  path=Fauna Jump.AnimGraphNode_StateMachine_0.DefaultJump.AnimStateNode_4.JumpApex
  nodes [Output Animation Pose(Result=Sequence Player 'Grandma_Jump_Apex'() {bLoopAnimation=false, sequence=&/Grandma_Jump_Apex})]
 }
 JumpFall {
  path=Fauna Jump.AnimGraphNode_StateMachine_0.DefaultJump.AnimStateNode_5.JumpFall
  nodes [Output Animation Pose(Result=Sequence Player 'Grandma_Jump_Fall_Loop'!(PlayRate:float=2) -> (Pose) {sequence=&Grandma_Jump_Fall_Loop})]
 }
 JumpLand {
  path=Fauna Jump.AnimGraphNode_StateMachine_0.DefaultJump.AnimStateNode_6.JumpLand
  nodes [Output Animation Pose(Result=Sequence Player 'Grandma_Jump_Land'() {bLoopAnimation=false, sequence=&Grandma_Jump_Land})]
 }
 AnimationTransitionGraph_0 {
  path=Fauna Jump.AnimGraphNode_StateMachine_0.DefaultJump.AnimStateTransitionNode_0.AnimationTransitionGraph_0
  nodes [Result(bCanEnterTransition=Get bStateRule_DefaultJump_ToJumpApex() -> (bStateRule_DefaultJump_ToJumpApex:bool))]
 }
 AnimationTransitionGraph_0 {
  path=Fauna Jump.AnimGraphNode_StateMachine_0.DefaultJump.AnimStateTransitionNode_1.AnimationTransitionGraph_0
  nodes [Result(bCanEnterTransition=Get bStateRule_DefaultJump_ToJumpApex())]
 }
 AnimationTransitionGraph_1 {
  path=Fauna Jump.AnimGraphNode_StateMachine_0.DefaultJump.AnimStateTransitionNode_11.AnimationTransitionGraph_1
  nodes [Result(bCanEnterTransition=Get bIsMovingOnGroundORIsSurfaceSwimming() -> (bIsMovingOnGroundORIsSurfaceSwimming:bool))]
 }
 AnimationTransitionGraph_1 {
  path=Fauna Jump.AnimGraphNode_StateMachine_0.DefaultJump.AnimStateTransitionNode_12.AnimationTransitionGraph_1
  nodes [Result(bCanEnterTransition=Get bIsMovingOnGroundORIsSurfaceSwimming())]
 }
 AnimationTransitionGraph_1 {path=Fauna Jump.AnimGraphNode_StateMachine_0.DefaultJump.AnimStateTransitionNode_5.AnimationTransitionGraph_1, nodes [Result()]}
 AnimationTransitionGraph_1 {
  path=Fauna Jump.AnimGraphNode_StateMachine_0.DefaultJump.AnimStateTransitionNode_6.AnimationTransitionGraph_1
  nodes [
   Result()
   Get bIsFalling()
  ]
 }
 AnimationTransitionGraph_1 {
  path=Fauna Jump.AnimGraphNode_StateMachine_0.DefaultJump.AnimStateTransitionNode_8.AnimationTransitionGraph_1
  nodes [Result(bCanEnterTransition=Get bIsMovingOnGroundORIsSurfaceSwimming())]
 }
 GetSequencerAnimSlotNames {nodes [
   GetSequencerAnimSlotNames() {is=title K2Node_FunctionEntry}
   > Return Node!(execute, ReturnValue:Array of Names=Make Array("[0]":Name=SequencerDefault) -> (Array:Array of Names))
  ]}
]
variables []
