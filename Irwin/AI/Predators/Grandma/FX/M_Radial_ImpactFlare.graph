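# Unlit, two-sided translucent Niagara sprite material for a radial impact flare in the Grandma predator effects.
# Particle color, scaled by a V gradient and the Color dynamic parameter, feeds the emissive adjustment function.
# Particle alpha masks a streak texture, which Breakup blends toward multiplication by a second tiled copy that Erosion shifts along V and V Scale stretches.
# A near-camera fade with distance and power parameters multiplies the resulting opacity.
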Material /Irwin/AI/Predators/Grandma/FX/M_Radial_ImpactFlare
graph {
 properties {
  bUsedWithNiagaraSprites=true
  blendMode=BLEND_Translucent
  pixelDepthOffsetMode=PDOM_Legacy
  shadingModel=MSM_Unlit
  twoSided=true
  usedShadingModels=Unlit
 }
 nodes [
  n1 = ParticleColor()
  n2 = DynamicParameter() {defaultValue=<rgba 1 0.5 0.75 1>, paramNames [Color, V Scale, Erosion, Breakup]}
  n3 = ComponentMask(None=TextureCoordinate()) {g=true}
  MaterialFunctionCall(Emissive=LinearInterpolate(B=@n1.RGB, Alpha=Multiply(A=@n2.Color, B=@n3))) {materialFunction=&MF_FX_EmissiveAdjustments}
  n4 = Multiply(A=@n1.A, B=TextureSample() {texture=/Game/Effects/Fort_Effects/Gameplay/SmokeDustExplosions/Textures/T_Explosion_Streak_02}.RGB)
  n5 = TextureCoordinate()
  n6 = DynamicParameter() {defaultValue=<rgba 1 0.5 0.75 1>, paramNames [Color, V Scale, Erosion, Breakup]}
  n7 = AppendVector(A=Constant() {r=4}, B=@n6.V Scale)
  n8 = Constant()
  n9 = LinearInterpolate(Alpha=@n6.Erosion) {constA=0.15, constB=0.75}
  n10 = TextureSample(UVs=Add(A=Multiply(A=@n5, B=@n7), B=AppendVector(A=@n8, B=@n9))) {texture=&T_Explosion_Streak_02}
  n11 = MaterialFunctionCall(Distance=ScalarParameter() {defaultValue=512, group=Opacity|Near_Camera_Fade, parameterName=NearCamFade_Distance, sortPriority=1}) {materialFunction=&Near_Camera_Fade}
  n12 = ScalarParameter() {defaultValue=1, group=Opacity|Near_Camera_Fade, parameterName=NearCamFade_Power, sortPriority=2}
  Multiply(
   A=LinearInterpolate(
    A=@n4
    B=Multiply(A=@n4, B=@n10.R)
    Alpha=DynamicParameter() {defaultValue=<rgba 1 0.5 0.75 1>, paramNames [Color, V Scale, Erosion, Breakup]}.Breakup
   ) {constA=1}
   B=Clamp(None=Power(Base=@n11.Result, Exp=@n12)) {minDefault=0.02}
  )
 ]
}
