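# Niagara system that gives the Grandma predator glowing idle eyes, drawn purely with a light renderer in a user-set eye color.
# A single burst spawns two persistent particles, each using its spawn group index to follow a bone or socket on the skeletal mesh.
# LOD-distance curves raise color brightness and sprite size farther away, and a time-of-day scalar adjusts light color intensity, strongest at night.
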NiagaraSystem /Irwin/AI/Predators/Grandma/FX/NS_Irwin_Grandma_IdleEye
dependencies {usedDataInterfaces [&NiagaraDataInterfaceCurve], usedDynamicInputs [], usedModules [&SystemState]}
emitters {Soft_Eye_Glow {
  emitterUpdateScript [
   &EmitterState(
    Inactive Response=::NewEnumerator1
    Life Cycle Mode=::NewEnumerator1
    Loop Behavior=::NewEnumerator1
    Loop Duration Mode=::NewEnumerator1
    UseLoopDelay=false
   )
   &SpawnBurst_Instantaneous(Spawn Count=2, Use Spawn Probability=false)
  ]
  particleSpawnScript [
   &FN_SetSpawnGroup(Number of Spawn Groups=2)
   &InitializeParticle(Color=$User.EyeColor, Lifetime=2, Sprite Size Mode=::NewEnumerator3, UsePositionOffset=false, Write Material Random=true)
   &CameraOffset(Camera Offset Amount=5)
  ]
  particleUpdateScript [
   &ParticleState(Kill Particles When Lifetime Has Elapsed=false, Let Infinitely Lived Particles Die When Emitter Deactivates=true)
   &SkeletalMeshLocation(
    Bone / Socket Index=$Particles.SpawnGroup
    Inherit Velocity=false
    Interpolate Mesh Sampling=false
    MaskSpawnGroup=false
    Mesh Sampling Type=::NewEnumerator5
    Skeletal Mesh=$User.SkeletalMesh
    Socket Sampling Mode=::NewEnumerator5
   )
   &ScaleSpriteSize(Scale Factor=&Vector2DFromFloat(Value=/Niagara/DynamicInputs/Scalability/ScalabilityDistanceBasedFloat(
      Curve:NiagaraDataInterfaceCurve=DataInterface() {propertyValues {Curve {keys [<at 0 1 RCIM_Cubic>, {arriveTangent=0.0019, interpMode=RCIM_Cubic, leaveTangent=0.0019, tangentMode=RCTM_User, time=5000, value=4}]}}}
      CurveIndex:NiagaraFloat=$Engine.Owner.LODDistance
      Bias:NiagaraFloat=0
      Scale:NiagaraFloat=1
     )))
   &ScaleColor(
    Scale RGB=&VectorFromFloat(Value=&ScalabilityDistanceBasedFloat(
      Curve:NiagaraDataInterfaceCurve=DataInterface() {propertyValues {Curve {keys [<at 0 1 RCIM_Cubic>,<at 5000 3 RCIM_Cubic>]}}}
      CurveIndex:NiagaraFloat=$Engine.Owner.LODDistance
      Bias:NiagaraFloat=0
      Scale:NiagaraFloat=1
     ))
    ScaleA=true
    ScaleRGB=true
   )
   &FN_Light_Attributes(
    Color=$Particles.Color
    Create Color Attribute=true
    Create Exponent Attribute=true
    Create Radius Attribute=true
    Exponent=3
    Radius=4
    Scale Color=&FNTimeOfDayScalar(Night Scale=20, Morning Scale=10, Day Scale=15, Evening Scale=10)
   )
  ]
  rendererClasses [&NiagaraLightRendererProperties]
  renderers [{rendererClass=&NiagaraLightRendererProperties, rendererIndex=0}]
 }}
summary {
 systemName=NS_Irwin_Grandma_IdleEye
 emitters [{emitterName=Soft_Eye_Glow, rendererClasses [&NiagaraLightRendererProperties]}]
 userVariables [
  User.EyeColor {type=LinearColor, description="", defaultValue {struct=LinearColor, value=<rgba 2 1.1667 0 1>}}
  User.SkeletalMesh {
   type=NiagaraDataInterfaceSkeletalMesh
   defaultValue {
    struct=NiagaraExt_VariableValue_DataInterface
    value {dataInterface=&self:NiagaraDataInterfaceSkeletalMesh_0}
   }
  }
 ]
}
