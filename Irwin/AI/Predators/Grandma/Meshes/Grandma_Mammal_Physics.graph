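# Physics asset for the mammal mesh of the Grandma predator AI, giving its skeleton a coarse collision shape.
# Eight single-capsule bodies cover the head, chest, pelvis, both front elbows, both rear ankles and one tail bone.
# Left and right limb capsules match in size with mirrored offsets.
# No constraints link the bodies, so the asset defines collision volumes without joint limits for a jointed ragdoll.
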PhysicsAsset /Irwin/AI/Predators/Grandma/Meshes/Grandma_Mammal_Physics
bodies {
 BipedHeadNeck_A_Head_C [{shapeType=Capsule, length=29.7715, radius=20.2802, center=<11.2315 4.6396 0.2864>, rotation=<rot 0.0851 125.4805 -89.5173>}]
 PawedArm_A_Elbow_L [{shapeType=Capsule, length=42.75, radius=12.25, center=<22.9875 -3.8664 -1.3364>, rotation=<rot -0.31 88.81 -89.31>}]
 PawedArm_A_Elbow_R [{shapeType=Capsule, length=42.75, radius=12.25, center=<-22.9693 4.745 1.3318>, rotation=<rot -0.31 88.81 90.69>}]
 PawedLeg_A_Ankle_L [{shapeType=Capsule, length=36.75, radius=13.25, center=<18.9787 6.0115 -0.0002>, rotation=<rot 90 -1.1935 -6.1935>}]
 PawedLeg_A_Ankle_R [{shapeType=Capsule, length=36.75, radius=13.25, center=<-18.9787 -6.0115 -0.0002>, rotation=<rot 90 -14.0363 160.9638>}]
 QuadSpine_A_Chest_C [{shapeType=Capsule, length=13.75, radius=30.25, center=<14.1743 4.4815 0.0003>, rotation=<rot 90 14.0363 -30.9614>}]
 QuadSpine_A_Pelvis_C [{shapeType=Capsule, length=15, radius=30, center=<12.36 4.44 0>, rotation=<rot 90 53.97 63.97>}]
 Tail_A_Tail4_C [{shapeType=Capsule, length=68, radius=17, center=<-4.07 -0.33 -0.58>, rotation=<rot 0 91.16 -88.75>}]
}
constraints []
